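# One-shot stereo MetaSound source that plays the Wandering Boss death sound in DemonMaskFire.
# On play, a random get picks from the wave array, which defaults to one death asset, and feeds a stereo wave player that signals finish.
# It uses its own attenuation, the boss-fight sound class and a stop-oldest concurrency override.
# A distance-scaled send to a music high-pass sidechain bus fades out as distance grows.
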MetaSoundSource /DemonMaskFire/Assets/Audio/DemonMaskFire_WanderingBoss/MSS_WanderingBoss_Death
inputs [
 UE.Source.OnPlay {type=Trigger}
 Wave Asset {type=WaveAsset:Array}
]
interfaces [UE.OutputFormat.Stereo:1.0, UE.Source.OneShot:1.0, UE.Source:1.1]
outputs [
 "UE.OutputFormat.Stereo.Audio:0" {type=Audio}
 "UE.OutputFormat.Stereo.Audio:1" {type=Audio}
 UE.Source.OneShot.OnFinished {type=Trigger}
]
pages [Default {
  input_defaults {UE.Source.OnPlay=null, Wave Asset [&/WanderingBoss_Death]}
  nodes [
   ArrayRandomGet_0 = Array.Random Get:WaveAsset:Array(Next=^UE.Source.OnPlay, In Array=^Wave Asset)
   WavePlayer_0 = UE.Wave Player:Stereo(Play=@ArrayRandomGet_0.On Next, Wave Asset=@ArrayRandomGet_0.Value)
  ]
  links [
   {dst="@UE.Source.OneShot.OnFinished", src=WavePlayer_0.On Finished}
   {dst="@UE.OutputFormat.Stereo.Audio:0", src=WavePlayer_0.Out Left}
   {dst="@UE.OutputFormat.Stereo.Audio:1", src=WavePlayer_0.Out Right}
  ]
 }]
properties {
 assetImportData=&self:AssetImportData
 attenuationSettings=&/Att_WanderingBoss_Death
 bOverrideConcurrency=true
 maxDistance=18000
 numChannels=2
 outputFormat=Stereo
 pitch=0.9
 sampleRate=48000
 soundClassObject=&Fort_BossFight
 volume=2.5
 busSends [{
   audioBus=/Game/Sounds/Modulation/EnvelopeFollowers/HPF/AB_SideChain_FortMusic_HPF
   maxSendDistance=9000
   maxSendLevel=0
   minSendDistance=5000
   minSendLevel=0.3
   sendLevel=0
  }]
 concurrencyOverrides {maxCount=3, resolutionRule=StopOldest}
}
