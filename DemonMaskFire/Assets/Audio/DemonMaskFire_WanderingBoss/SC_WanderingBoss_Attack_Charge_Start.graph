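# Sound cue for the start of the Wandering Boss's charge attack in DemonMaskFire.
# A single wave player feeds a modulator that varies pitch and volume on each play, with the pitch minimum raised to unshifted.
# The cue uses the boss's melee attack attenuation and the boss fight SFX sound class, with boosted volume.
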SoundCue /DemonMaskFire/Assets/Audio/DemonMaskFire_WanderingBoss/SC_WanderingBoss_Attack_Charge_Start
root=Modulator_0
nodes [
 {id=Modulator_0, children [WavePlayer_0], properties {pitchMin=1}}
 {id=WavePlayer_0, properties {soundWaveAssetPtr=&/WanderingBoss_Attack_Charge_Start_01}}
]
properties {
 attenuationSettings=&/SA_WanderingBoss_Melee_Attack
 maxDistance=7500
 soundClassObject=&Fort_BossFight_SFX
 volumeMultiplier=1.5
}
