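# Sound Cue for the sweetener layer of the Demon Mask Fire wandering boss's charge attack.
# A single wave player feeds a Modulator at its default settings for slight pitch and volume variation.
# It plays louder than default through the boss fight SFX class, using the boss's melee attack attenuation.
# A concurrency override allows one instance per owner, stopping the oldest on retrigger.
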SoundCue /DemonMaskFire/Assets/Audio/DemonMaskFire_WanderingBoss/SC_WanderingBoss_Attack_Charge_Sweetener
root=Modulator_2
nodes [
 {id=Modulator_2, children [WavePlayer_0]}
 {
  id=WavePlayer_0
  properties {soundWaveAssetPtr=&/WanderingBoss_Attack_Charge_Sweetener}
 }
]
properties {
 attenuationSettings=&/SA_WanderingBoss_Melee_Attack
 bOverrideConcurrency=true
 maxDistance=7500
 soundClassObject=&Fort_BossFight_SFX
 volumeMultiplier=1.5
 busSends [{audioBus=null, minSendLevel=0, sendLevel=1, soundSourceBus=null}]
 concurrencyOverrides {bLimitToOwner=true, maxCount=1, resolutionRule=StopOldest}
}
