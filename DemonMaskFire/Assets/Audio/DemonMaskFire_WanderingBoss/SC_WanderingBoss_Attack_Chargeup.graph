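# Sound cue for the Wandering Boss attack charge-up in the DemonMaskFire content.
# A single wave player plays one charge sound through a modulator left at default settings, with raised cue volume.
# It uses the boss melee attack attenuation and the Fort boss fight SFX sound class.
# A concurrency override allows one instance per owner, stopping the oldest when a new one starts.
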SoundCue /DemonMaskFire/Assets/Audio/DemonMaskFire_WanderingBoss/SC_WanderingBoss_Attack_Chargeup
root=Modulator_2
nodes [
 {id=Modulator_2, children [WavePlayer_1]}
 {id=WavePlayer_1, properties {soundWaveAssetPtr=&/WanderingBoss_Attack_Charge_01}}
]
properties {
 attenuationSettings=&/SA_WanderingBoss_Melee_Attack
 bOverrideConcurrency=true
 maxDistance=7500
 soundClassObject=&Fort_BossFight_SFX
 volumeMultiplier=1.5
 concurrencyOverrides {bLimitToOwner=true, maxCount=1, resolutionRule=StopOldest}
}
