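# Sound cue for the build-up sound of the Demon Mask Fire wandering boss's barrage, classed as boss fight SFX.
# A Random node picks one of four equally weighted waves and feeds a Modulator at the root for per-play variation.
# It uses an emitter attenuation asset, caps concurrency with farthest-first stopping, and keeps playing when silent.
# Distance-scaled sends to music and weapon sidechain buses are strongest up close, with weapons sent harder than music.
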SoundCue /DemonMaskFire/Assets/Audio/DemonMaskFire_WanderingBoss/SC_WanderingBoss_Barrage_Build
root=Modulator_1
nodes [
 {id=Modulator_1, children [Random_0]}
 {id=Random_0, children [WavePlayer_4,WavePlayer_5,WavePlayer_6,WavePlayer_7], properties {weights [1,1,1,1]}}
 {id=WavePlayer_4, properties {soundWaveAssetPtr=&/Mask_Demon_Fire_Barrage_Build_04}}
 {id=WavePlayer_5, properties {soundWaveAssetPtr=&/Mask_Demon_Fire_Barrage_Build_03}}
 {id=WavePlayer_6, properties {soundWaveAssetPtr=&/Mask_Demon_Fire_Barrage_Build_02}}
 {id=WavePlayer_7, properties {soundWaveAssetPtr=&/Mask_Demon_Fire_Barrage_Build_01}}
]
properties {
 attenuationSettings=&/SA_WanderingBoss_Barrage_Emitter
 bOverrideConcurrency=true
 maxDistance=5500
 soundClassObject=&Fort_BossFight_SFX
 virtualizationMode=PlayWhenSilent
 volumeMultiplier=1
 busSends [
  {
   audioBus=&AB_Sidechain_FortMusic_Vol
   maxSendDistance=2500
   maxSendLevel=0
   minSendDistance=0
   minSendLevel=0.3
   sendLevel=0
   soundSourceBus=null
   sourceBusSendLevelControlMethod=Linear
  }
  {
   audioBus=&AB_Sidechain_FortWeapon_Vol
   maxSendDistance=2500
   maxSendLevel=0
   minSendDistance=0
   minSendLevel=1
   sendLevel=0
   soundSourceBus=null
   sourceBusSendLevelControlMethod=Linear
  }
 ]
 concurrencyOverrides {maxCount=4, resolutionRule=StopFarthestThenPreventNew}
}
