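# Sound cue for the end of the Demon Mask Fire wandering boss's barrage attack.
# A single Mask_Demon_Fire_Barrage_End wave plays through a Modulator node, with the cue's volume boosted.
# It routes to the boss-fight SFX sound class and reuses the player Demon Mask Fire attenuation settings rather than a boss-specific one.
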SoundCue /DemonMaskFire/Assets/Audio/DemonMaskFire_WanderingBoss/SC_WanderingBoss_Barrage_End
root=Modulator_0
nodes [
 {id=Modulator_0, children [WavePlayer_1]}
 {id=WavePlayer_1, properties {soundWaveAssetPtr=&/Mask_Demon_Fire_Barrage_End_01}}
]
properties {
 attenuationSettings=/DemonMaskFire/Assets/Audio/DemonMaskFire_Player/SA_Mask_Demon_Fire_Player
 maxDistance=10400
 soundClassObject=&Fort_BossFight_SFX
 volumeMultiplier=1.5
}
