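# Sound cue for the impact of the Demon Mask Fire wandering boss's barrage attack.
# A Random node picks one of four equally weighted impact waves, feeding a Modulator at the root.
# Playback uses a barrage explosion attenuation asset, the boss-fight SFX class and an overridden concurrency limit.
# Distance-scaled bus sends feed the music and weapon volume sidechain buses, fading to zero as distance increases.
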SoundCue /DemonMaskFire/Assets/Audio/DemonMaskFire_WanderingBoss/SC_WanderingBoss_Barrage_Impact
root=Modulator_0
nodes [
 {id=Modulator_0, children [Random_1]}
 {id=Random_1, children [WavePlayer_0,WavePlayer_1,WavePlayer_2,WavePlayer_3], properties {weights [1,1,1,1]}}
 {id=WavePlayer_0, properties {soundWaveAssetPtr=&/Mask_Demon_Fire_Barrage_Impact_01}}
 {id=WavePlayer_1, properties {soundWaveAssetPtr=&/Mask_Demon_Fire_Barrage_Impact_02}}
 {id=WavePlayer_2, properties {soundWaveAssetPtr=&/Mask_Demon_Fire_Barrage_Impact_03}}
 {id=WavePlayer_3, properties {soundWaveAssetPtr=&/Mask_Demon_Fire_Barrage_Impact_04}}
]
properties {
 attenuationSettings=&/SA_WanderingBoss_Barrage_Explo
 bOverrideConcurrency=true
 maxDistance=8000
 soundClassObject=&Fort_BossFight_SFX
 volumeMultiplier=1.5
 busSends [
  {
   audioBus=&AB_Sidechain_FortMusic_Vol
   maxSendDistance=6000
   maxSendLevel=0
   minSendDistance=1000
   minSendLevel=0.3
   sendLevel=0
   soundSourceBus=null
   sourceBusSendLevelControlMethod=Linear
  }
  {
   audioBus=&AB_Sidechain_FortWeapon_Vol
   maxSendDistance=5000
   maxSendLevel=0
   minSendDistance=2500
   minSendLevel=1
   sendLevel=0
   soundSourceBus=null
   sourceBusSendLevelControlMethod=Linear
  }
 ]
 concurrencyOverrides {maxCount=4}
}
