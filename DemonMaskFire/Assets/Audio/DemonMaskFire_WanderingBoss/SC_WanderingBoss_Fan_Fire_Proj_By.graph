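# Sound cue for the Wandering Boss fan fire attack, apparently the close fly-by whoosh of its fire projectiles.
# A Random node picks among four equally weighted fly-by waves borrowed from the Demon Mask Fire player audio, which also supplies the attenuation asset.
# It plays in the boss-fight SFX class at reduced volume, with overridden concurrency that caps simultaneous instances and blocks new ones to limit stacking.
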SoundCue /DemonMaskFire/Assets/Audio/DemonMaskFire_WanderingBoss/SC_WanderingBoss_Fan_Fire_Proj_By
root=Random_1
nodes [
 {id=Random_1, children [WavePlayer_0,WavePlayer_1,WavePlayer_2,WavePlayer_3], properties {weights [1,1,1,1]}}
 {id=WavePlayer_0, properties {soundWaveAssetPtr=/DemonMaskFire/Assets/Audio/DemonMaskFire_Player/Mask_Demon_Fire_Proj_By_Close_01}}
 {id=WavePlayer_1, properties {soundWaveAssetPtr=/DemonMaskFire/Assets/Audio/DemonMaskFire_Player/Mask_Demon_Fire_Proj_By_Close_02}}
 {id=WavePlayer_2, properties {soundWaveAssetPtr=/DemonMaskFire/Assets/Audio/DemonMaskFire_Player/Mask_Demon_Fire_Proj_By_Close_03}}
 {id=WavePlayer_3, properties {soundWaveAssetPtr=/DemonMaskFire/Assets/Audio/DemonMaskFire_Player/Mask_Demon_Fire_Proj_By_Close_04}}
]
properties {
 attenuationSettings=/DemonMaskFire/Assets/Audio/DemonMaskFire_Player/SA_Mask_Demon_Fire_Player
 bOverrideConcurrency=true
 maxDistance=10400
 soundClassObject=&Fort_BossFight_SFX
 volumeMultiplier=0.7
 concurrencyOverrides {maxCount=2, resolutionRule=PreventNew, retriggerTime=0.1}
}
