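# Sound cue for the looping fire projectile sound of the Wandering Boss's fan attack in DemonMaskFire.
# A looping wave player reuses the player Demon Mask fire projectile loop, fed through an Enveloper that fades volume in over a quarter second.
# Boss projectile attenuation and the boss fight SFX class apply, with a concurrency override capping simultaneous instances at two with a short retrigger time.
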SoundCue /DemonMaskFire/Assets/Audio/DemonMaskFire_WanderingBoss/SC_WanderingBoss_Fan_Fire_Proj_Loop
root=Enveloper_0
nodes [
 {
  id=Enveloper_0
  children [WavePlayer_0]
  properties {pitchCurve {editorCurveData {keys [{time=0}]}}, volumeCurve {editorCurveData {keys [<at 0 0>,<at 0.0959 0.2448>,<at 0.25 1>]}}}
 }
 {id=WavePlayer_0, properties {bLooping=true, soundWaveAssetPtr=/DemonMaskFire/Assets/Audio/DemonMaskFire_Player/Mask_Demon_Fire_Proj_Loop_01}}
]
properties {
 attenuationSettings=&/SA_WanderingBoss_Proj
 bDebug=true
 bOverrideConcurrency=true
 maxDistance=6500
 soundClassObject=&Fort_BossFight_SFX
 volumeMultiplier=1
 concurrencyOverrides {maxCount=2, retriggerTime=0.1}
}
