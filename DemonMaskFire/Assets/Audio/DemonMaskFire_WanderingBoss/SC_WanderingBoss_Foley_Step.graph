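# Sound cue for the footstep foley of the Wandering Boss in DemonMaskFire.
# A single random node picks among twelve equally weighted step wave variations, with no pitch or volume modulation nodes.
# It plays at doubled volume through the boss-fight foley sound class, spatialized by the Wandering Boss foley attenuation asset.
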SoundCue /DemonMaskFire/Assets/Audio/DemonMaskFire_WanderingBoss/SC_WanderingBoss_Foley_Step
root=Random_0
nodes [
 {
  id=Random_0
  children [
   WavePlayer_0
   WavePlayer_1
   WavePlayer_2
   WavePlayer_3
   WavePlayer_4
   WavePlayer_5
   WavePlayer_6
   WavePlayer_7
   WavePlayer_8
   WavePlayer_9
   WavePlayer_10
   WavePlayer_11
  ]
  properties {weights [1,1,1,1,1,1,1,1,1,1,1,1]}
 }
 {id=WavePlayer_0, properties {soundWaveAssetPtr=&/WanderingBoss_Foley_Step_12}}
 {id=WavePlayer_1, properties {soundWaveAssetPtr=&/WanderingBoss_Foley_Step_11}}
 {id=WavePlayer_10, properties {soundWaveAssetPtr=&/WanderingBoss_Foley_Step_02}}
 {id=WavePlayer_11, properties {soundWaveAssetPtr=&/WanderingBoss_Foley_Step_01}}
 {id=WavePlayer_2, properties {soundWaveAssetPtr=&/WanderingBoss_Foley_Step_10}}
 {id=WavePlayer_3, properties {soundWaveAssetPtr=&/WanderingBoss_Foley_Step_09}}
 {id=WavePlayer_4, properties {soundWaveAssetPtr=&/WanderingBoss_Foley_Step_08}}
 {id=WavePlayer_5, properties {soundWaveAssetPtr=&/WanderingBoss_Foley_Step_07}}
 {id=WavePlayer_6, properties {soundWaveAssetPtr=&/WanderingBoss_Foley_Step_06}}
 {id=WavePlayer_7, properties {soundWaveAssetPtr=&/WanderingBoss_Foley_Step_05}}
 {id=WavePlayer_8, properties {soundWaveAssetPtr=&/WanderingBoss_Foley_Step_04}}
 {id=WavePlayer_9, properties {soundWaveAssetPtr=&/WanderingBoss_Foley_Step_03}}
]
properties {
 attenuationSettings=&/SA_WanderingBoss_Foley
 maxDistance=7000
 soundClassObject=/Game/Sounds/FortSoundClassesAndModes/Fort_BossFight_SoundClass/Fort_BossFight_Foley
 volumeMultiplier=2
}
