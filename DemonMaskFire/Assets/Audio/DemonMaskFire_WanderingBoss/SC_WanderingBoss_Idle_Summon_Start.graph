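# Sound Cue for the Wandering Boss idle summon start, layering a Syphon start one-shot into a sustained Syphon loop.
# A Mixer plays the start wave through a Modulator at default settings alongside the looping wave, which sits behind a one-second Delay under an Enveloper with a rising volume curve.
# It uses the boss fight SFX sound class and the boss idle attenuation settings.
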SoundCue /DemonMaskFire/Assets/Audio/DemonMaskFire_WanderingBoss/SC_WanderingBoss_Idle_Summon_Start
root=Mixer_0
nodes [
 {id=Delay_1, children [WavePlayer_2], properties {delayMax=1, delayMin=1}}
 {
  id=Enveloper_0
  children [Delay_1]
  properties {pitchCurve {editorCurveData {keys [{time=0}]}}, volumeCurve {editorCurveData {keys [<at 0 0>,<at 1 1>]}}}
 }
 {id=Mixer_0, children [Modulator_0,Enveloper_0], properties {inputVolume [1,1]}}
 {id=Modulator_0, children [WavePlayer_1]}
 {id=WavePlayer_1, properties {soundWaveAssetPtr=&/WanderingBoss_Idle_Syphon_Start_01}}
 {
  id=WavePlayer_2
  properties {bLooping=true, soundWaveAssetPtr=&/WanderingBoss_Idle_Syphon_Loop_01}
 }
]
properties {
 attenuationSettings=&/SA_WanderingBoss_Idle
 maxDistance=9500
 soundClassObject=&Fort_BossFight_SFX
 volumeMultiplier=1
}
