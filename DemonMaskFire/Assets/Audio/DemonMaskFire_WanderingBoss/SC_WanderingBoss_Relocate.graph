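# Sound cue for the relocate sound of the Wandering Boss in DemonMaskFire.
# A single wave player feeds a Modulator left at default settings as the root.
# It uses the boss's fire attenuation asset, routes to the boss fight SFX sound class, and boosts volume above unity.
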SoundCue /DemonMaskFire/Assets/Audio/DemonMaskFire_WanderingBoss/SC_WanderingBoss_Relocate
root=Modulator_0
nodes [
 {id=Modulator_0, children [WavePlayer_1]}
 {id=WavePlayer_1, properties {soundWaveAssetPtr=&/WanderingBoss_Relocate_01}}
]
properties {
 attenuationSettings=&/SA_WanderingBoss_Fire
 maxDistance=14000
 soundClassObject=&Fort_BossFight_SFX
 volumeMultiplier=1.5
}
