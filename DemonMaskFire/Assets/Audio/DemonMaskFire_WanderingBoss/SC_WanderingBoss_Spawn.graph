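# Sound cue for the Wandering Boss spawn sound in DemonMaskFire.
# A single spawn wave plays through a Modulator node left at its default settings.
# The cue routes to the boss-fight SFX sound class with a raised volume and uses the fire boss attenuation asset for long-range spatial falloff.
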SoundCue /DemonMaskFire/Assets/Audio/DemonMaskFire_WanderingBoss/SC_WanderingBoss_Spawn
root=Modulator_0
nodes [
 {id=Modulator_0, children [WavePlayer_4]}
 {id=WavePlayer_4, properties {soundWaveAssetPtr=&/WanderingBoss_Spawn_01}}
]
properties {
 attenuationSettings=&/SA_WanderingBoss_Fire
 maxDistance=14000
 soundClassObject=&Fort_BossFight_SFX
 volumeMultiplier=1.5
}
