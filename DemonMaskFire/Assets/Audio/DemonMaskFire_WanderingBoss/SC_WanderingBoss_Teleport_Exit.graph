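# Sound cue for the Wandering Boss teleport exit effect in DemonMaskFire.
# A Random node picks one of three equally weighted wave variations and passes it through a Modulator at the root.
# It plays in the boss-fight SFX sound class with the Wandering Boss fire attenuation asset.
# A concurrency override caps it at two simultaneous instances.
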SoundCue /DemonMaskFire/Assets/Audio/DemonMaskFire_WanderingBoss/SC_WanderingBoss_Teleport_Exit
root=Modulator_0
nodes [
 {id=Modulator_0, children [Random_0]}
 {id=Random_0, children [WavePlayer_7,WavePlayer_8,WavePlayer_9], properties {weights [1,1,1]}}
 {id=WavePlayer_7, properties {soundWaveAssetPtr=&/WanderingBoss_Teleport_Exit_01}}
 {id=WavePlayer_8, properties {soundWaveAssetPtr=&/WanderingBoss_Teleport_Exit_02}}
 {id=WavePlayer_9, properties {soundWaveAssetPtr=&/WanderingBoss_Teleport_Exit_03}}
]
properties {
 attenuationSettings=&/SA_WanderingBoss_Fire
 bOverrideConcurrency=true
 maxDistance=14000
 soundClassObject=&Fort_BossFight_SFX
 volumeMultiplier=1
 attenuationOverrides {bAttenuate=false}
 concurrencyOverrides {maxCount=2}
}
